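# Physics asset for the left-arm dynamic bones of the ShyTurkey male medium player body cosmetic.
# A chain of nine equal-sized capsules, dyn_arm_l_01 to 09, hangs from upperarm_l, each link constrained to the previous with very tight swing and twist limits.
# The last link also joins hand_l with free swing and twist, tying the chain at both ends.
# A tiny lowerarm_l body is present without constraints.
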PhysicsAsset /BRCosmetics/Characters/Player/Male/Medium/Bodies/M_MED_ShyTurkey/Meshes/M_MED_ShyTurkey_Arm_L_Physics
bodies {
 dyn_arm_l_01 [{shapeType=Capsule, length=6, radius=4, shapeName=dyn_arm_l_01_capsule, center=<2.7808 0 0>, rotation=<rot 90 0 0>}]
 dyn_arm_l_02 [{shapeType=Capsule, length=6, radius=4, shapeName=dyn_arm_l_02_capsule, center=<4.536 0 0>, rotation=<rot 90 0 0>}]
 dyn_arm_l_03 [{shapeType=Capsule, length=6, radius=4, shapeName=dyn_arm_l_03_capsule, center=<5.2256 0 0>, rotation=<rot 90 0 0>}]
 dyn_arm_l_04 [{shapeType=Capsule, length=6, radius=4, shapeName=dyn_arm_l_04_capsule, center=<4.5946 0 0>, rotation=<rot 90 0 0>}]
 dyn_arm_l_05 [{shapeType=Capsule, length=6, radius=4, shapeName=dyn_arm_l_05_capsule, rotation=<rot 90 0 0>}]
 dyn_arm_l_06 [{shapeType=Capsule, length=6, radius=4, shapeName=dyn_arm_l_05_capsule, center=<1.5 0 0>, rotation=<rot 90 0 0>}]
 dyn_arm_l_07 [{shapeType=Capsule, length=6, radius=4, shapeName=dyn_arm_l_05_capsule, center=<1.5 0 0>, rotation=<rot 90 0 0>}]
 dyn_arm_l_08 [{shapeType=Capsule, length=6, radius=4, shapeName=dyn_arm_l_05_capsule, center=<1.5 0 0>, rotation=<rot 90 0 0>}]
 dyn_arm_l_09 [{shapeType=Capsule, length=6, radius=4, shapeName=dyn_arm_l_05_capsule, center=<1.5 0 0>, rotation=<rot 90 0 0>}]
 hand_l [{shapeType=Capsule, length=5.036, radius=3.018, shapeName=hand_l_capsule, center=<1.905 0.0593 1.4065>, rotation=<rot 2.0999 -90 3.0928>}]
 lowerarm_l [{shapeType=Capsule, length=0.505, radius=0.505, shapeName=lowerarm_l_capsule, rotation=<rot 0 -90 0>}]
 upperarm_l [{shapeType=Capsule, length=0.505, radius=0.505, shapeName=upperarm_l_capsule, rotation=<rot 0 -90 0>}]
}
constraints [
 {bone1Name=dyn_arm_l_01, bone2Name=upperarm_l, swing1LimitDegrees=1, swing2LimitDegrees=1, twistLimitDegrees=2}
 {bone1Name=dyn_arm_l_02, bone2Name=dyn_arm_l_01, swing1LimitDegrees=1, swing2LimitDegrees=1, twistLimitDegrees=2}
 {bone1Name=dyn_arm_l_03, bone2Name=dyn_arm_l_02, swing1LimitDegrees=1, swing2LimitDegrees=1, twistLimitDegrees=2}
 {bone1Name=dyn_arm_l_04, bone2Name=dyn_arm_l_03, swing1LimitDegrees=1, swing2LimitDegrees=1, twistLimitDegrees=1}
 {bone1Name=dyn_arm_l_05, bone2Name=dyn_arm_l_04, swing1LimitDegrees=1, swing2LimitDegrees=1, twistLimitDegrees=1}
 {bone1Name=dyn_arm_l_06, bone2Name=dyn_arm_l_05, swing1LimitDegrees=1, swing2LimitDegrees=1, twistLimitDegrees=1}
 {bone1Name=dyn_arm_l_07, bone2Name=dyn_arm_l_06, swing1LimitDegrees=1, swing2LimitDegrees=1, twistLimitDegrees=1}
 {bone1Name=dyn_arm_l_08, bone2Name=dyn_arm_l_07, swing1LimitDegrees=1, swing2LimitDegrees=1, twistLimitDegrees=1}
 {bone1Name=dyn_arm_l_09, bone2Name=dyn_arm_l_08, swing1LimitDegrees=1, swing2LimitDegrees=1, twistLimitDegrees=1}
 {
  bone1Name=dyn_arm_l_09
  bone2Name=hand_l
  swing1LimitDegrees=45
  swing1Motion=Free
  swing2LimitDegrees=45
  swing2Motion=Free
  twistLimitDegrees=45
  twistMotion=Free
 }
]
