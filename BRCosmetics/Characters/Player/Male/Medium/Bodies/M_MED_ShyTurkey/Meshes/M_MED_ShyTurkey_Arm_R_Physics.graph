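# Physics asset for the right arm of the ShyTurkey medium male player body, simulating a nine-bone dyn_arm_r chain strung between upperarm_r and hand_r.
# Equal-sized capsules cover the chain, with small capsules on the two end bones.
# Very tight angular limits link each segment back to upperarm_r, while the last segment joins hand_r with free swing and twist.
# Chain capsule names carry left-arm labels, several repeated.
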PhysicsAsset /BRCosmetics/Characters/Player/Male/Medium/Bodies/M_MED_ShyTurkey/Meshes/M_MED_ShyTurkey_Arm_R_Physics
bodies {
 dyn_arm_r_01 [{shapeType=Capsule, length=6, radius=4, shapeName=dyn_arm_l_01_capsule, center=<2.7808 0 0>, rotation=<rot 90 0 0>}]
 dyn_arm_r_02 [{shapeType=Capsule, length=6, radius=4, shapeName=dyn_arm_l_02_capsule, center=<4.536 0 0>, rotation=<rot 90 0 0>}]
 dyn_arm_r_03 [{shapeType=Capsule, length=6, radius=4, shapeName=dyn_arm_l_03_capsule, center=<5.2256 0 0>, rotation=<rot 90 0 0>}]
 dyn_arm_r_04 [{shapeType=Capsule, length=6, radius=4, shapeName=dyn_arm_l_04_capsule, center=<4.5946 0 0>, rotation=<rot 90 0 0>}]
 dyn_arm_r_05 [{shapeType=Capsule, length=6, radius=4, shapeName=dyn_arm_l_05_capsule, rotation=<rot 90 0 0>}]
 dyn_arm_r_06 [{shapeType=Capsule, length=6, radius=4, shapeName=dyn_arm_l_05_capsule, center=<1.5 0 0>, rotation=<rot 90 0 0>}]
 dyn_arm_r_07 [{shapeType=Capsule, length=6, radius=4, shapeName=dyn_arm_l_05_capsule, center=<1.5 0 0>, rotation=<rot 90 0 0>}]
 dyn_arm_r_08 [{shapeType=Capsule, length=6, radius=4, shapeName=dyn_arm_l_05_capsule, center=<1.5 0 0>, rotation=<rot 90 0 0>}]
 dyn_arm_r_09 [{shapeType=Capsule, length=6, radius=4, shapeName=dyn_arm_l_05_capsule, center=<1.5 0 0>, rotation=<rot 90 0 0>}]
 hand_r [{shapeType=Capsule, length=1.3669, radius=1.3497, shapeName=dyn_arm_r_01_capsule, center=<2.7808 0 0>, rotation=<rot 90 0 0>}]
 upperarm_r [{shapeType=Capsule, length=0.505, radius=0.505, shapeName=upperarm_r_capsule, rotation=<rot 0 -90 0>}]
}
constraints [
 {bone1Name=dyn_arm_r_01, bone2Name=upperarm_r, swing1LimitDegrees=1, swing2LimitDegrees=1, twistLimitDegrees=2}
 {bone1Name=dyn_arm_r_02, bone2Name=dyn_arm_r_01, swing1LimitDegrees=1, swing2LimitDegrees=1, twistLimitDegrees=2}
 {bone1Name=dyn_arm_r_03, bone2Name=dyn_arm_r_02, swing1LimitDegrees=1, swing2LimitDegrees=1, twistLimitDegrees=1}
 {bone1Name=dyn_arm_r_04, bone2Name=dyn_arm_r_03, swing1LimitDegrees=1, swing2LimitDegrees=1, twistLimitDegrees=1}
 {bone1Name=dyn_arm_r_05, bone2Name=dyn_arm_r_04, swing1LimitDegrees=1, swing2LimitDegrees=1, twistLimitDegrees=1}
 {bone1Name=dyn_arm_r_06, bone2Name=dyn_arm_r_05, swing1LimitDegrees=1, swing2LimitDegrees=1, twistLimitDegrees=1}
 {bone1Name=dyn_arm_r_07, bone2Name=dyn_arm_r_06, swing1LimitDegrees=1, swing2LimitDegrees=1, twistLimitDegrees=1}
 {bone1Name=dyn_arm_r_08, bone2Name=dyn_arm_r_07, swing1LimitDegrees=1, swing2LimitDegrees=1, twistLimitDegrees=1}
 {bone1Name=dyn_arm_r_09, bone2Name=dyn_arm_r_08, swing1LimitDegrees=1, swing2LimitDegrees=1, twistLimitDegrees=1}
 {
  bone1Name=dyn_arm_r_09
  bone2Name=hand_r
  swing1LimitDegrees=45
  swing1Motion=Free
  swing2LimitDegrees=45
  swing2Motion=Free
  twistLimitDegrees=45
  twistMotion=Free
 }
]
