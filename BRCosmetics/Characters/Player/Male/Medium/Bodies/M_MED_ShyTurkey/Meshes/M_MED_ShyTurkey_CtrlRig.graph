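# Control Rig for the male medium ShyTurkey player body that drives deform bones plus dynamic bow and arm bones.
# The shared DeformRig function runs first, then dyn_bow_main follows weighted upper-body bones and fixed transforms adjust its bow and string masters.
# A sequence then parents each arm's nine nulls to upperarm or lowerarm, and stretchy B-spline IK fits the arm bone chains to their controls.
# An alternative weighted right-arm constraint chain is left unconnected.
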ControlRigBlueprint /BRCosmetics/Characters/Player/Male/Medium/Bodies/M_MED_ShyTurkey/Meshes/M_MED_ShyTurkey_CtrlRig
graphs [
 RigVMFunctionLibrary {links [], nodes []}
 RigVMModel {nodes [
   &RigUnit_ParentConstraint(
    AdvancedSettings=(RotationOrderForFilter=XZY)
    Child=(Type=Null,Name="dyn_arm_r_01_null")
    ExecutePin=()
    Parents="((Item=(Type=Bone,Name=\"pinky_metacarpal_r\"),Weight=0.0066),(Item=(Type=Bone,Name=\"ring_metacarpal_r\"),Weight=0.0085),(Item=(Type=Bone,Name=\"thumb_01_r\"),Weight=0.0227),(Item=(Type=Bone,Name=\"index_metacarpal_r\"),Weight=0.0269),(Item=(Type=Bone,Name=\"lower ...(truncated)"
    Weight=1
   )
   > &RigUnit_ParentConstraint(
    AdvancedSettings=(RotationOrderForFilter=XZY)
    Child=(Type=Null,Name="dyn_arm_r_02_null")
    Parents="((Item=(Type=Bone,Name=\"index_metacarpal_r\"),Weight=0.0041),(Item=(Type=Bone,Name=\"ring_metacarpal_r\"),Weight=0.0046),(Item=(Type=Bone,Name=\"lowerarm_r\"),Weight=0.0142),(Item=(Type=Bone,Name=\"pinky_metacarpal_r\"),Weight=0.0151),(Item=(Type=Bone,Name=\"hand_ ...(truncated)"
    Weight=1
   )
   > &RigUnit_ParentConstraint(
    AdvancedSettings=(RotationOrderForFilter=XZY)
    Child=(Type=Null,Name="dyn_arm_r_03_null")
    Parents="((Item=(Type=Bone,Name=\"deform_elbow_in_r\"),Weight=0.0006),(Item=(Type=Bone,Name=\"upperarm_r\"),Weight=0.0025),(Item=(Type=Bone,Name=\"lowerarm_r\"),Weight=0.179),(Item=(Type=Bone,Name=\"lowerarm_twist_01_r\"),Weight=0.2972),(Item=(Type=Bone,Name=\"lowerarm_twis ...(truncated)"
    Weight=1
   )
   > &RigUnit_ParentConstraint(
    AdvancedSettings=(RotationOrderForFilter=XZY)
    Child=(Type=Null,Name="dyn_arm_r_04_null")
    Parents="((Item=(Type=Bone,Name=\"upperarm_twist_02_r\"),Weight=0.0017),(Item=(Type=Bone,Name=\"deform_elbow_in_r\"),Weight=0.0242),(Item=(Type=Bone,Name=\"deform_elbow_out_r\"),Weight=0.027),(Item=(Type=Bone,Name=\"upperarm_r\"),Weight=0.089),(Item=(Type=Bone,Name=\"lowera ...(truncated)"
    Weight=1
   )
   > &RigUnit_ParentConstraint(
    AdvancedSettings=(RotationOrderForFilter=XZY)
    Child=(Type=Null,Name="dyn_arm_r_05_null")
    Parents="((Item=(Type=Bone,Name=\"lowerarm_twist_01_r\"),Weight=0.005),(Item=(Type=Bone,Name=\"deform_elbow_in_r\"),Weight=0.009),(Item=(Type=Bone,Name=\"upperarm_twist_02_r\"),Weight=0.0316),(Item=(Type=Bone,Name=\"lowerarm_twist_02_r\"),Weight=0.1237),(Item=(Type=Bone,Na ...(truncated)"
    Weight=1
   )
   > &RigUnit_ParentConstraint(
    AdvancedSettings=(RotationOrderForFilter=XZY)
    Child=(Type=Null,Name="dyn_arm_r_06_null")
    Parents="((Item=(Type=Bone,Name=\"deform_elbow_in_r\"),Weight=0.0152),(Item=(Type=Bone,Name=\"deform_elbow_out_r\"),Weight=0.0169),(Item=(Type=Bone,Name=\"lowerarm_twist_02_r\"),Weight=0.0173),(Item=(Type=Bone,Name=\"upperarm_twist_01_r\"),Weight=0.0668),(Item=(Type=Bone,N ...(truncated)"
    Weight=1
   )
   > &RigUnit_ParentConstraint(
    AdvancedSettings=(RotationOrderForFilter=XZY)
    Child=(Type=Null,Name="dyn_arm_r_07_null")
    Parents="((Item=(Type=Bone,Name=\"clavicle_r\"),Weight=0.0039),(Item=(Type=Bone,Name=\"deform_upperarm_bk_r\"),Weight=0.0042),(Item=(Type=Bone,Name=\"deform_elbow_in_r\"),Weight=0.0109),(Item=(Type=Bone,Name=\"deform_upperarm_fr_r\"),Weight=0.0117),(Item=(Type=Bone,Name=\"l ...(truncated)"
    Weight=1
   )
   > &RigUnit_ParentConstraint(
    AdvancedSettings=(RotationOrderForFilter=XZY)
    Child=(Type=Null,Name="dyn_arm_r_08_null")
    Parents="((Item=(Type=Bone,Name=\"deform_upperarm_bk_r\"),Weight=0.0342),(Item=(Type=Bone,Name=\"clavicle_r\"),Weight=0.0567),(Item=(Type=Bone,Name=\"upperarm_r\"),Weight=0.086),(Item=(Type=Bone,Name=\"deform_upperarm_up_r\"),Weight=0.093),(Item=(Type=Bone,Name=\"upperarm_t ...(truncated)"
    Weight=1
   )
   > &RigUnit_ParentConstraint(
    AdvancedSettings=(RotationOrderForFilter=XZY)
    Child=(Type=Null,Name="dyn_arm_r_09_null")
    Parents="((Item=(Type=Bone,Name=\"deform_upperarm_bk_r\"),Weight=0.0134),(Item=(Type=Bone,Name=\"spine_03\"),Weight=0.0326),(Item=(Type=Bone,Name=\"deform_upperarm_fr_r\"),Weight=0.0377),(Item=(Type=Bone,Name=\"spine_05\"),Weight=0.0475),(Item=(Type=Bone,Name=\"upperarm_twi ...(truncated)"
    Weight=1
   )
   RigUnit_BeginExecution = &RigUnit_BeginExecution()
   FN_DeformRig_FNC = RigVMFunctionReferenceNode(
    DeformRig_ Curve="(deformRig_l_upperarm_up_alpha_72_#1=False,deformRig_r_upperarm_up_alpha_73_#2=False,deformRig_l_upperarm_fr_alpha_74_#3=False,deformRig_r_upperarm_fr_alpha_75_CB51B8 ...(truncated)"
    DeformRig_Bool="(upperarm_1_#4=True,elbow_3_#5=True,wrist_7_#6=True,groin_8_#7=True,glute_17_#8=True,knee_9_40C408324CA261 ...(truncated)"
    DeformRig_Debug=(upperarm_debug_18_#4=False,wrist_debug_20_#6=False,lat_debug_26_#9=False,groinGlute_debug_27_#7=False)
    DeformRig_Value="(upperarm_up_18_#10=(Rotation=(X=0,Y=0,Z=0,W=1),Translation=(X=0,Y=0,Z=8),Scale3D=(X=0.7,Y=1,Z=1)),upperarm_fwd_fr_19_#11=(Rotation=(X=0,Y=0,Z=0,W=1),Translation=(X=0,Y=10,Z=0),Scale3D=(X=0.5,Y=1,Z= ...(truncated)"
    ExecuteContext=@RigUnit_BeginExecution.ExecutePin
   )
   > &RigUnit_ParentConstraint(
    AdvancedSettings=(RotationOrderForFilter=XZY)
    Child=(Type=Bone,Name="dyn_bow_main")
    Parents="((Item=(Type=Bone,Name=\"clavicle_l\"),Weight=0.0003),(Item=(Type=Bone,Name=\"head\"),Weight=0.0089),(Item=(Type=Bone,Name=\"spine_04\"),Weight=0.0195),(Item=(Type=Bone,Name=\"neck_01\"),Weight=0.0538),(Item=(Type=Bone,Name=\"neck_02\"),Weight=0.1234),(Item=(Type=Bo ...(truncated)"
    Weight=1
   )
   > n1 = &RigUnit_ModifyTransforms(
    ItemToModify="((Item=(Type=Bone,Name=\"dyn_L_bow_MSTR\"),Transform=(Rotation=(X=0.0804,Y=-0.1991,Z=0.3657,W=0.9056),Translation=(X=0,Y=0,Z=0),Scale3D=(X=1,Y=1,Z=1))),(Item=(Type=Bone,Name=\"dyn_L_string_MSTR\"),Transform=(Rotation=(X=0.2745,Y=-0.4098,Z=-0.3784,W=0.7833),Tra ...(truncated)"
    Weight=1
    WeightMinimum=0
   )
   RigVMFunction_Sequence = &RigVMFunction_Sequence(ExecuteContext=@n1.ExecutePin)
   &RigUnit_ParentConstraint(
    AdvancedSettings=(InterpolationType=Average,RotationOrderForFilter=XZY)
    Child=(Type=Null,Name="dyn_arm_l_01_null")
    Parents=((Item=(Type=Bone,Name="upperarm_l"),Weight=1))
    Weight=1
    ExecutePin=@RigVMFunction_Sequence.A
   )
   > &RigUnit_ParentConstraint(
    AdvancedSettings=(InterpolationType=Average,RotationOrderForFilter=XZY)
    Child=(Type=Null,Name="dyn_arm_l_02_null")
    Parents=((Item=(Type=Bone,Name="upperarm_l"),Weight=1))
    Weight=1
   )
   > &RigUnit_ParentConstraint(
    AdvancedSettings=(InterpolationType=Average,RotationOrderForFilter=XZY)
    Child=(Type=Null,Name="dyn_arm_l_03_null")
    Parents=((Item=(Type=Bone,Name="upperarm_l"),Weight=1))
    Weight=1
   )
   > &RigUnit_ParentConstraint(
    AdvancedSettings=(InterpolationType=Average,RotationOrderForFilter=XZY)
    Child=(Type=Null,Name="dyn_arm_l_04_null")
    Parents=((Item=(Type=Bone,Name="upperarm_l"),Weight=1))
    Weight=1
   )
   > &RigUnit_ParentConstraint(
    AdvancedSettings=(InterpolationType=Average,RotationOrderForFilter=XZY)
    Child=(Type=Null,Name="dyn_arm_l_05_null")
    Parents=((Item=(Type=Bone,Name="lowerarm_l"),Weight=1))
    Weight=1
   )
   > &RigUnit_ParentConstraint(
    AdvancedSettings=(InterpolationType=Average,RotationOrderForFilter=XZY)
    Child=(Type=Null,Name="dyn_arm_l_06_null")
    Parents=((Item=(Type=Bone,Name="lowerarm_l"),Weight=1))
    Weight=1
   )
   > &RigUnit_ParentConstraint(
    AdvancedSettings=(InterpolationType=Average,RotationOrderForFilter=XZY)
    Child=(Type=Null,Name="dyn_arm_l_07_null")
    Parents=((Item=(Type=Bone,Name="lowerarm_l"),Weight=1))
    Weight=1
   )
   > &RigUnit_ParentConstraint(
    AdvancedSettings=(InterpolationType=Average,RotationOrderForFilter=XZY)
    Child=(Type=Null,Name="dyn_arm_l_08_null")
    Parents=((Item=(Type=Bone,Name="lowerarm_l"),Weight=1))
    Weight=1
   )
   > &RigUnit_ParentConstraint(
    AdvancedSettings=(InterpolationType=Average,RotationOrderForFilter=XZY)
    Child=(Type=Null,Name="dyn_arm_l_09_null")
    Parents=((Item=(Type=Bone,Name="lowerarm_l"),Weight=1))
    Weight=1
   )
   Arm_L_Bones = RigVMDispatchNode(Value="((Type=Bone,Name=\"dyn_arm_l_01\"),(Type=Bone,Name=\"dyn_arm_l_02\"),(Type=Bone,Name=\"dyn_arm_l_03\"),(Type=Bone,Name=\"dyn_arm_l_04\"),(Type=Bone,Name=\"dyn_arm_l_05\"),(Type=Bone,Name=\"dyn_arm_l_06\"),(Type=Bone,Name=\"dyn_arm_l_07\"),(Type=Bone,Name=\"dyn_arm_l_08\") ...(truncated)") {template="DISPATCH_RigVMDispatch_Constant(io Value)"}
   Arm_L_Controls = RigVMDispatchNode(Value="((Type=Control,Name=\"dyn_arm_l_01_ctrl\"),(Type=Control,Name=\"dyn_arm_l_02_ctrl\"),(Type=Control,Name=\"dyn_arm_l_03_ctrl\"),(Type=Control,Name=\"dyn_arm_l_04_ctrl\"),(Type=Control,Name=\"dyn_arm_l_05_ctrl\"),(Type=Control,Name=\"dyn_arm_l_06_ctrl\"),(Type=Control,N ...(truncated)") {template="DISPATCH_RigVMDispatch_Constant(io Value)"}
   > SplineIK = RigVMFunctionReferenceNode(
    Primary Axis=(X=1,Y=0,Z=0)
    Samples Per Segment=4
    Secondary Spline Direction=(X=0,Y=5,Z=0)
    Spline Mode=BSpline
    Stretch=true
    Up Axis=(X=0,Y=1,Z=0)
    debug=true
    Bones=@Arm_L_Bones.Value
    Controls=@Arm_L_Controls.Value
   )
   &RigUnit_ParentConstraint(
    AdvancedSettings=(InterpolationType=Average,RotationOrderForFilter=XZY)
    Child=(Type=Null,Name="dyn_arm_r_01_null")
    Parents=((Item=(Type=Bone,Name="upperarm_r"),Weight=1))
    Weight=1
    ExecutePin=@RigVMFunction_Sequence.B
   )
   > &RigUnit_ParentConstraint(
    AdvancedSettings=(InterpolationType=Average,RotationOrderForFilter=XZY)
    Child=(Type=Null,Name="dyn_arm_r_02_null")
    Parents=((Item=(Type=Bone,Name="upperarm_r"),Weight=1))
    Weight=1
   )
   > &RigUnit_ParentConstraint(
    AdvancedSettings=(InterpolationType=Average,RotationOrderForFilter=XZY)
    Child=(Type=Null,Name="dyn_arm_r_03_null")
    Parents=((Item=(Type=Bone,Name="upperarm_r"),Weight=1))
    Weight=1
   )
   > &RigUnit_ParentConstraint(
    AdvancedSettings=(InterpolationType=Average,RotationOrderForFilter=XZY)
    Child=(Type=Null,Name="dyn_arm_r_04_null")
    Parents=((Item=(Type=Bone,Name="upperarm_r"),Weight=1))
    Weight=1
   )
   > &RigUnit_ParentConstraint(
    AdvancedSettings=(InterpolationType=Average,RotationOrderForFilter=XZY)
    Child=(Type=Null,Name="dyn_arm_r_05_null")
    Parents=((Item=(Type=Bone,Name="lowerarm_r"),Weight=1))
    Weight=1
   )
   > &RigUnit_ParentConstraint(
    AdvancedSettings=(InterpolationType=Average,RotationOrderForFilter=XZY)
    Child=(Type=Null,Name="dyn_arm_r_06_null")
    Parents=((Item=(Type=Bone,Name="lowerarm_r"),Weight=1))
    Weight=1
   )
   > &RigUnit_ParentConstraint(
    AdvancedSettings=(InterpolationType=Average,RotationOrderForFilter=XZY)
    Child=(Type=Null,Name="dyn_arm_r_07_null")
    Parents=((Item=(Type=Bone,Name="lowerarm_r"),Weight=1))
    Weight=1
   )
   > &RigUnit_ParentConstraint(
    AdvancedSettings=(InterpolationType=Average,RotationOrderForFilter=XZY)
    Child=(Type=Null,Name="dyn_arm_r_08_null")
    Parents=((Item=(Type=Bone,Name="lowerarm_r"),Weight=1))
    Weight=1
   )
   > &RigUnit_ParentConstraint(
    AdvancedSettings=(InterpolationType=Average,RotationOrderForFilter=XZY)
    Child=(Type=Null,Name="dyn_arm_r_09_null")
    Parents=((Item=(Type=Bone,Name="lowerarm_r"),Weight=1))
    Weight=1
   )
   Arm_R_Bones = RigVMDispatchNode(Value="((Type=Bone,Name=\"dyn_arm_r_01\"),(Type=Bone,Name=\"dyn_arm_r_02\"),(Type=Bone,Name=\"dyn_arm_r_03\"),(Type=Bone,Name=\"dyn_arm_r_04\"),(Type=Bone,Name=\"dyn_arm_r_05\"),(Type=Bone,Name=\"dyn_arm_r_06\"),(Type=Bone,Name=\"dyn_arm_r_07\"),(Type=Bone,Name=\"dyn_arm_r_08\") ...(truncated)") {template="DISPATCH_RigVMDispatch_Constant(io Value)"}
   Arm_R_Controls = RigVMDispatchNode(Value="((Type=Control,Name=\"dyn_arm_r_01_ctrl\"),(Type=Control,Name=\"dyn_arm_r_02_ctrl\"),(Type=Control,Name=\"dyn_arm_r_03_ctrl\"),(Type=Control,Name=\"dyn_arm_r_04_ctrl\"),(Type=Control,Name=\"dyn_arm_r_05_ctrl\"),(Type=Control,Name=\"dyn_arm_r_06_ctrl\"),(Type=Control,N ...(truncated)") {template="DISPATCH_RigVMDispatch_Constant(io Value)"}
   > SplineIK_1 = RigVMFunctionReferenceNode(
    Primary Axis=(X=1,Y=0,Z=0)
    Samples Per Segment=4
    Secondary Spline Direction=(X=0,Y=5,Z=0)
    Spline Mode=BSpline
    Stretch=true
    Up Axis=(X=0,Y=1,Z=0)
    debug=true
    Bones=@Arm_R_Bones.Value
    Controls=@Arm_R_Controls.Value
   )
  ]}
]
hierarchy [
 root {loc=<0 0 0>, rot=<0 0 0 1>}
 pelvis {parent=root, loc=<0 -0.7071 93.1191>, rot=<0.0127 0.707 -0.0127 -0.707>}
 spine_01 {parent=pelvis, loc=<6.7494 -0.6388 0>, rot=<0 0 0.034 -0.9994>}
 spine_02 {parent=spine_01, loc=<7.2518 -0.035 0>, rot=<0 0 0.0238 -0.9997>}
 spine_03 {parent=spine_02, loc=<6.0104 0 0>, rot=<0 0 0.0149 -0.9999>}
 spine_04 {parent=spine_03, loc=<8.1024 0 0>, rot=<0 0 -0.0132 -0.9999>}
 spine_05 {parent=spine_04, loc=<10.0146 0 0>, rot=<0 0 -0.0015 -1>}
 neck_01 {parent=spine_05, loc=<11.3138 0.1529 0>, rot=<0 0 -0.2003 -0.9797>}
 neck_02 {parent=neck_01, loc=<4.5627 0 0>, rot=<0 0 0.0504 -0.9987>}
 head {parent=neck_02, loc=<5.3914 0 0>, rot=<0 0 0.115 -0.9934>}
 attach_cape {parent=spine_05, loc=<9.0175 -11.714 0>, rot=<0.0616 -0.7044 -0.0616 -0.7044>}
 cape {parent=attach_cape, type=Null, loc=<0 0 0>, rot=<0 0 0 1>}
 attach_backpack {parent=spine_05, loc=<-0.1224 -13.473 0>, rot=<-0.0093 -0.707 0.0093 -0.707>}
 Backpack_BR {parent=attach_backpack, type=Null, loc=<0 4.1468 -4.4883>, rot=<-0.0871 0 0 0.9962>}
 clavicle_l {parent=spine_05, loc=<4.3575 -0.1466 -2.3223>, rot=<0.0465 -0.7289 -0.009 -0.6829>}
 upperarm_l {parent=clavicle_l, loc=<15.5111 0 0>, rot=<0.0226 -0.3234 -0.0088 -0.9459>}
 lowerarm_l {parent=upperarm_l, loc=<26.6514 0 0>, rot=<0.0391 -0.0099 -0.2453 -0.9686>}
 hand_l {parent=lowerarm_l, loc=<22.3491 0 0>, rot=<-0.7081 -0.0507 0.0975 -0.6975>}
 thumb_01_l {parent=hand_l, loc=<1.3118 -0.3191 -2.459>, rot=<0.6546 -0.4042 0.0664 -0.6354>}
 thumb_02_l {parent=thumb_01_l, loc=<5.9336 -0.0448 0.0401>, rot=<-0.0037 0.0213 0.298 -0.9543>}
 thumb_03_l {parent=thumb_02_l, loc=<2.2223 0.3682 -0.0907>, rot=<0.0118 -0.013 0.0066 -0.9998>}
 index_metacarpal_l {parent=hand_l, loc=<1.6535 0.0183 -2.96>, rot=<0.0158 -0.0124 -0.0173 -0.9996>}
 index_01_l {parent=index_metacarpal_l, loc=<8.3008 0.0357 -0.0524>, rot=<-0.0044 0.0485 0.0924 -0.9945>}
 index_02_l {parent=index_01_l, loc=<3.3587 -0.3004 0.0452>, rot=<-0.002 0.0027 0.3225 -0.9466>}
 index_03_l {parent=index_02_l, loc=<2.5578 -0.123 -0.0252>, rot=<-0.002 0.0067 0.2558 -0.9667>}
 middle_metacarpal_l {parent=hand_l, loc=<2.1856 -0.1725 -0.5984>, rot=<0.006 0.0168 -0.0329 -0.9993>}
 middle_01_l {parent=middle_metacarpal_l, loc=<7.3275 -0.2447 0.0129>, rot=<-0.0208 0.0042 0.0858 -0.9961>}
 middle_02_l {parent=middle_01_l, loc=<4.0123 0.1401 0.1175>, rot=<-0.0218 0.0258 0.3512 -0.9357>}
 middle_03_l {parent=middle_02_l, loc=<2.6077 0.0534 0.1373>, rot=<0.0082 0.0188 0.1708 -0.9851>}
 ring_metacarpal_l {parent=hand_l, loc=<2.2074 -0.3625 1.3638>, rot=<-0.0284 0.0722 -0.0458 -0.9959>}
 ring_01_l {parent=ring_metacarpal_l, loc=<7.0295 -0.6197 -0.1446>, rot=<-0.0489 -0.0252 0.1892 -0.9804>}
 ring_02_l {parent=ring_01_l, loc=<3.6174 0.4647 -0.1027>, rot=<0 0 0.2806 -0.9598>}
 ring_03_l {parent=ring_02_l, loc=<2.4152 0.114 0.0199>, rot=<0 0 0.1189 -0.9929>}
 pinky_metacarpal_l {parent=hand_l, loc=<1.7616 -0.9753 3.3814>, rot=<-0.0397 0.0814 -0.069 -0.9935>}
 pinky_01_l {parent=pinky_metacarpal_l, loc=<7.1709 -0.4831 -0.1094>, rot=<-0.0654 -0.0035 0.2242 -0.9723>}
 pinky_02_l {parent=pinky_01_l, loc=<2.8274 0.0893 -0.1043>, rot=<0.0037 0.0011 0.2854 -0.9584>}
 pinky_03_l {parent=pinky_02_l, loc=<2.2046 0.0307 -0.0078>, rot=<-0.0002 0.0098 0.1445 -0.9895>}
 deform_wrist_dn_base_l {parent=hand_l, loc=<0 -2.2342 0>, rot=<0 0 0 -1>}
 deform_wrist_dn_l {parent=deform_wrist_dn_base_l, loc=<0 0 0>, rot=<0 0 0 1>}
 deform_wrist_up_base_l {parent=hand_l, loc=<0 2.2682 0>, rot=<0 0 0 -1>}
 deform_wrist_up_l {parent=deform_wrist_up_base_l, loc=<0 0 0>, rot=<0 0 0 1>}
 lowerarm_twist_02_l {parent=lowerarm_l, loc=<9.6199 0 0>, rot=<0 0 0 -1>}
 lowerarm_twist_01_l {parent=lowerarm_l, loc=<17.8549 0 0>, rot=<0 0 0 -1>}
 deform_elbow_in_base_l {parent=lowerarm_l, loc=<0.2112 5.259 0.8274>, rot=<0 0 0 -1>}
 deform_elbow_in_l {parent=deform_elbow_in_base_l, loc=<0 0 0>, rot=<0 0 0 1>}
 deform_elbow_out_base_l {parent=lowerarm_l, loc=<0 -4 0>, rot=<0 0 0 -1>}
 deform_elbow_out_l {parent=deform_elbow_out_base_l, loc=<0 0 0>, rot=<0 0 0 1>}
 upperarm_twist_01_l {parent=upperarm_l, loc=<0.0041 0 0>, rot=<0 0 0 -1>}
 deform_upperarm_bk_base_l {parent=upperarm_twist_01_l, loc=<1.0813 -7.2193 -1.9453>, rot=<0 0 0 1>}
 deform_upperarm_bk_l {parent=deform_upperarm_bk_base_l, loc=<0 0 0>, rot=<0 0 0 1>}
 deform_upperarm_fr_base_l {parent=upperarm_twist_01_l, loc=<-0.3691 7.0506 -2.765>, rot=<0 0 0 1>}
 deform_upperarm_fr_l {parent=deform_upperarm_fr_base_l, loc=<0 0 0>, rot=<0 0 0 1>}
 deform_upperarm_up_base_l {parent=upperarm_twist_01_l, loc=<-3.2216 -1.1391 3.3311>, rot=<0 0 0 1>}
 deform_upperarm_up_l {parent=deform_upperarm_up_base_l, loc=<0 0 0>, rot=<0 0 0 1>}
 upperarm_twist_02_l {parent=upperarm_l, loc=<13.7941 0 0>, rot=<0 0 0 -1>}
 dyn_arm_l_01 {parent=upperarm_l, loc=<0 0 0>, rot=<0 0 0 -1>}
 dyn_arm_l_02 {parent=dyn_arm_l_01, loc=<6.0544 0 0>, rot=<0 0 0 1>}
 dyn_arm_l_03 {parent=dyn_arm_l_02, loc=<6.9081 0 0>, rot=<0 0 0 1>}
 dyn_arm_l_04 {parent=dyn_arm_l_03, loc=<7.3467 0 0>, rot=<0 0 0 1>}
 dyn_arm_l_05 {parent=dyn_arm_l_04, loc=<6.3421 0 0>, rot=<0.0594 -0.0047 -0.2454 -0.9676>}
 dyn_arm_l_06 {parent=dyn_arm_l_05, loc=<6.2652 0 0>, rot=<0 0 0 -1>}
 dyn_arm_l_07 {parent=dyn_arm_l_06, loc=<5.7567 0 0>, rot=<0 0 0 1>}
 dyn_arm_l_08 {parent=dyn_arm_l_07, loc=<5.1365 0 0>, rot=<0 0 0 1>}
 dyn_arm_l_09 {parent=dyn_arm_l_08, loc=<5.1907 0 0>, rot=<0 -0.0557 0 -0.9984>}
 clavicle_r {parent=spine_05, loc=<4.3575 -0.1466 2.3223>, rot=<0.7289 0.0465 -0.6829 0.009>}
 upperarm_r {parent=clavicle_r, loc=<-15.5111 0 0>, rot=<0.0226 -0.3234 -0.0088 -0.9459>}
 lowerarm_r {parent=upperarm_r, loc=<-26.6514 0 0>, rot=<0.0391 -0.0099 -0.2453 -0.9686>}
 hand_r {parent=lowerarm_r, loc=<-22.3491 0 0>, rot=<-0.7081 -0.0507 0.0975 -0.6975>}
 thumb_01_r {parent=hand_r, loc=<-1.3118 0.3191 2.459>, rot=<0.6546 -0.4042 0.0664 -0.6354>}
 thumb_02_r {parent=thumb_01_r, loc=<-5.9336 0.0448 -0.0401>, rot=<-0.0037 0.0213 0.298 -0.9543>}
 thumb_03_r {parent=thumb_02_r, loc=<-2.2223 -0.3682 0.0907>, rot=<0.0118 -0.013 0.0066 -0.9998>}
 index_metacarpal_r {parent=hand_r, loc=<-1.6535 -0.0183 2.96>, rot=<0.0158 -0.0124 -0.0173 -0.9996>}
 index_01_r {parent=index_metacarpal_r, loc=<-8.3008 -0.0357 0.0524>, rot=<-0.0044 0.0485 0.0924 -0.9945>}
 index_02_r {parent=index_01_r, loc=<-3.3587 0.3004 -0.0452>, rot=<-0.002 0.0027 0.3225 -0.9466>}
 index_03_r {parent=index_02_r, loc=<-2.5578 0.123 0.0252>, rot=<-0.002 0.0067 0.2558 -0.9667>}
 middle_metacarpal_r {parent=hand_r, loc=<-2.1856 0.1725 0.5984>, rot=<0.006 0.0168 -0.0329 -0.9993>}
 middle_01_r {parent=middle_metacarpal_r, loc=<-7.3275 0.2447 -0.0129>, rot=<-0.0208 0.0042 0.0858 -0.9961>}
 middle_02_r {parent=middle_01_r, loc=<-4.0123 -0.1401 -0.1175>, rot=<-0.0218 0.0258 0.3512 -0.9357>}
 middle_03_r {parent=middle_02_r, loc=<-2.6077 -0.0534 -0.1373>, rot=<0.0082 0.0188 0.1708 -0.9851>}
 ring_metacarpal_r {parent=hand_r, loc=<-2.2074 0.3625 -1.3638>, rot=<-0.0284 0.0722 -0.0458 -0.9959>}
 ring_01_r {parent=ring_metacarpal_r, loc=<-7.0295 0.6197 0.1446>, rot=<-0.0489 -0.0252 0.1892 -0.9804>}
 ring_02_r {parent=ring_01_r, loc=<-3.6174 -0.4647 0.1027>, rot=<0 0 0.2806 -0.9598>}
 ring_03_r {parent=ring_02_r, loc=<-2.4152 -0.114 -0.0199>, rot=<0 0 0.1189 -0.9929>}
 pinky_metacarpal_r {parent=hand_r, loc=<-1.7616 0.9753 -3.3814>, rot=<-0.0397 0.0814 -0.069 -0.9935>}
 pinky_01_r {parent=pinky_metacarpal_r, loc=<-7.1709 0.4831 0.1094>, rot=<-0.0654 -0.0035 0.2242 -0.9723>}
 pinky_02_r {parent=pinky_01_r, loc=<-2.8274 -0.0893 0.1043>, rot=<0.0037 0.0011 0.2854 -0.9584>}
 pinky_03_r {parent=pinky_02_r, loc=<-2.2046 -0.0307 0.0078>, rot=<-0.0002 0.0098 0.1445 -0.9895>}
 deform_wrist_dn_base_r {parent=hand_r, loc=<-0.0004 2.2345 -0.0001>, rot=<0 0 0 -1>}
 deform_wrist_dn_r {parent=deform_wrist_dn_base_r, loc=<0 0 0>, rot=<0 0 0 1>}
 deform_wrist_up_base_r {parent=hand_r, loc=<0.0003 -2.2684 0.0001>, rot=<0 0 0 -1>}
 deform_wrist_up_r {parent=deform_wrist_up_base_r, loc=<0 0 0>, rot=<0 0 0 1>}
 lowerarm_twist_02_r {parent=lowerarm_r, loc=<-9.6199 0 0>, rot=<0 0 0 -1>}
 lowerarm_twist_01_r {parent=lowerarm_r, loc=<-17.8549 0 0>, rot=<0 0 0 -1>}
 deform_elbow_in_base_r {parent=lowerarm_r, loc=<-0.2111 -5.2591 -0.8275>, rot=<0 0 0 -1>}
 deform_elbow_in_r {parent=deform_elbow_in_base_r, loc=<-0.0001 0 0.0002>, rot=<0 0 0 1>}
 deform_elbow_out_base_r {parent=lowerarm_r, loc=<0 4 -0.0001>, rot=<0 0 0 -1>}
 deform_elbow_out_r {parent=deform_elbow_out_base_r, loc=<0 0 0>, rot=<0 0 0 1>}
 upperarm_twist_01_r {parent=upperarm_r, loc=<-0.0041 0 0>, rot=<0 0 0 -1>}
 deform_upperarm_bk_base_r {parent=upperarm_twist_01_r, loc=<-1.0814 7.2192 1.9455>, rot=<0 0 0 1>}
 deform_upperarm_bk_r {parent=deform_upperarm_bk_base_r, loc=<0 0 0>, rot=<0 0 0 1>}
 deform_upperarm_fr_base_r {parent=upperarm_twist_01_r, loc=<0.3689 -7.0506 2.7652>, rot=<0 0 0 1>}
 deform_upperarm_fr_r {parent=deform_upperarm_fr_base_r, loc=<0 0 0>, rot=<0 0 0 1>}
 deform_upperarm_up_base_r {parent=upperarm_twist_01_r, loc=<3.2214 1.1391 -3.3308>, rot=<0 0 0 1>}
 deform_upperarm_up_r {parent=deform_upperarm_up_base_r, loc=<0 0 0>, rot=<0 0 0 -1>}
 upperarm_twist_02_r {parent=upperarm_r, loc=<-13.7941 0 0>, rot=<0 0 0 -1>}
 dyn_arm_r_01 {parent=upperarm_r, loc=<-0.0001 0 0.0002>, rot=<0 -0.9997 -0.0233 0>}
 dyn_arm_r_02 {parent=dyn_arm_r_01, loc=<6.0544 0 0>, rot=<0 0 0 1>}
 dyn_arm_r_03 {parent=dyn_arm_r_02, loc=<6.9082 0 0>, rot=<0 0 0 1>}
 dyn_arm_r_04 {parent=dyn_arm_r_03, loc=<7.3467 0 0>, rot=<0 0 0 1>}
 dyn_arm_r_05 {parent=dyn_arm_r_04, loc=<6.3421 0 0>, rot=<-0.0074 -0.0293 0.2437 -0.9694>}
 dyn_arm_r_06 {parent=dyn_arm_r_05, loc=<6.2648 0 0>, rot=<0 -0.0001 0 -1>}
 dyn_arm_r_07 {parent=dyn_arm_r_06, loc=<5.7569 0 0>, rot=<0 0 0 -1>}
 dyn_arm_r_08 {parent=dyn_arm_r_07, loc=<5.1368 0 0>, rot=<0 0.0001 0 -1>}
 dyn_arm_r_09 {parent=dyn_arm_r_08, loc=<5.1903 0 0>, rot=<0 0 0 1>}
 dyn_bow_main {parent=spine_05, loc=<7.704 7.3951 0.0229>, rot=<0.0284 -0.7065 -0.0284 -0.7065>}
 dyn_L_string_MSTR {parent=dyn_bow_main, loc=<0.8709 0.0774 0.1199>, rot=<0.547 -0.2087 -0.5352 -0.6089>}
 dyn_L_string_1 {parent=dyn_L_string_MSTR, loc=<0 0 0>, rot=<0.547 -0.2087 -0.5352 -0.6089>}
 dyn_L_string_2 {parent=dyn_L_string_1, loc=<3.4267 0 0>, rot=<-0.0104 0.0124 -0.1239 -0.9922>}
 dyn_L_string_3 {parent=dyn_L_string_2, loc=<4.2572 0 0>, rot=<0.0309 -0.0059 -0.0193 -0.9993>}
 dyn_L_bow_MSTR {parent=dyn_bow_main, loc=<2.5361 -0.2051 0.0327>, rot=<-0.0863 -0.0546 0.07 -0.9923>}
 dyn_L_bow_1 {parent=dyn_L_bow_MSTR, loc=<0 0 0>, rot=<-0.0863 -0.0546 0.07 -0.9923>}
 dyn_L_bow_2 {parent=dyn_L_bow_1, loc=<2.5195 0 0>, rot=<0 0 0 1>}
 dyn_R_string_MSTR {parent=dyn_bow_main, loc=<-1.003 0.2885 -0.0607>, rot=<0.2388 -0.4774 -0.6476 -0.5437>}
 dyn_R_string_1 {parent=dyn_R_string_MSTR, loc=<0 0 0>, rot=<0.2388 -0.4774 -0.6476 -0.5437>}
 dyn_R_string_2 {parent=dyn_R_string_1, loc=<3.2288 0 0>, rot=<-0.1195 -0.0232 0.1542 -0.9805>}
 dyn_R_string_3 {parent=dyn_R_string_2, loc=<2.9172 0 0>, rot=<-0.0289 -0.0184 0.0218 -0.9992>}
 dyn_R_bow_MSTR {parent=dyn_bow_main, loc=<-2.8965 -0.1342 -0.2316>, rot=<0.0665 -0.8293 -0.5493 -0.078>}
 dyn_R_bow_1 {parent=dyn_R_bow_MSTR, loc=<0 0 0>, rot=<0.0665 -0.8293 -0.5493 -0.078>}
 dyn_R_bow_2 {parent=dyn_R_bow_1, loc=<2.0619 0 0>, rot=<0 0 0 1>}
 deform_pec_aim_base_l {parent=spine_04, loc=<7.8732 12.6388 -2.4123>, rot=<0.0102 -0.4602 0.3979 -0.7936>}
 deform_pec_aim_l {parent=deform_pec_aim_base_l, loc=<0 0 0>, rot=<0 0 0 1>}
 deform_pec_base_l {parent=deform_pec_aim_l, loc=<10.766 0 0>, rot=<0 0 0 1>}
 deform_pec_l {parent=deform_pec_base_l, loc=<0.055 0 0>, rot=<0 0 0 1>}
 deform_pec_aim_base_r {parent=spine_04, loc=<7.8727 12.6388 2.4123>, rot=<-0.4602 -0.0102 0.7936 0.3979>}
 deform_pec_aim_r {parent=deform_pec_aim_base_r, loc=<0 0 0>, rot=<0 0 0 1>}
 deform_pec_base_r {parent=deform_pec_aim_r, loc=<-10.7662 -0.0006 -0.0007>, rot=<0 0 0 1>}
 deform_pec_r {parent=deform_pec_base_r, loc=<0 0 0>, rot=<0 0 0 1>}
 thigh_l {parent=pelvis, loc=<-2.5693 -0.0109 -7.7501>, rot=<0.0012 0.0572 0.0155 -0.9982>}
 calf_l {parent=thigh_l, loc=<-39.0484 0 0>, rot=<0.0006 -0.0161 -0.0391 -0.9991>}
 foot_l {parent=calf_l, loc=<-42.561 0.1077 0.4292>, rot=<0.0418 -0.0455 0.0343 -0.9975>}
 ball_l {parent=foot_l, loc=<-6.1599 13.8264 0.6806>, rot=<0.0001 -0.0001 -0.7063 -0.7079>}
 calf_twist_02_l {parent=calf_l, loc=<-18.8064 -0.0967 0.0397>, rot=<0.0237 0.0095 -0.0035 -0.9997>}
 calf_twist_01_l {parent=calf_l, loc=<-35.813 -0.0967 0.0397>, rot=<0.0237 0.0095 -0.0035 -0.9997>}
 deform_knee_in_base_l {parent=calf_l, loc=<1.298 -5.2957 -0.0496>, rot=<0 0 0 -1>}
 deform_knee_in_l {parent=deform_knee_in_base_l, loc=<0 0 0>, rot=<0 0 0 1>}
 deform_knee_out_base_l {parent=calf_l, loc=<0 4 0>, rot=<0 0 0 -1>}
 deform_knee_out_l {parent=deform_knee_out_base_l, loc=<0 0 0>, rot=<0 0 0 1>}
 thigh_twist_01_l {parent=thigh_l, loc=<-21.6303 0 0>, rot=<0.0013 0 0 -1>}
 deform_groin_base_l {parent=thigh_l, loc=<-2.1908 8.1142 -1.7442>, rot=<0 0 0 -1>}
 deform_groin_l {parent=deform_groin_base_l, loc=<0 0 0>, rot=<0 0 0 1>}
 thigh_r {parent=pelvis, loc=<-2.5693 -0.0109 7.7501>, rot=<0.0572 -0.0012 0.9982 0.0155>}
 calf_r {parent=thigh_r, loc=<39.0484 0 0>, rot=<0.0006 -0.0161 -0.0391 -0.9991>}
 foot_r {parent=calf_r, loc=<42.561 -0.1077 -0.4292>, rot=<0.0418 -0.0455 0.0343 -0.9975>}
 ball_r {parent=foot_r, loc=<6.1599 -13.8264 -0.6806>, rot=<0.0001 -0.0001 -0.7063 -0.7079>}
 calf_twist_02_r {parent=calf_r, loc=<18.8064 0.0967 -0.0397>, rot=<0.0237 0.0095 -0.0035 -0.9997>}
 calf_twist_01_r {parent=calf_r, loc=<35.813 0.0967 -0.0397>, rot=<0.0237 0.0095 -0.0035 -0.9997>}
 deform_knee_in_base_r {parent=calf_r, loc=<-1.298 5.2957 0.0496>, rot=<0 0 0 -1>}
 deform_knee_in_r {parent=deform_knee_in_base_r, loc=<0 0 0>, rot=<0 0 0 1>}
 deform_knee_out_base_r {parent=calf_r, loc=<0 -4 0>, rot=<0 0 0 -1>}
 deform_knee_out_r {parent=deform_knee_out_base_r, loc=<0 0 0>, rot=<0 0 0 1>}
 thigh_twist_01_r {parent=thigh_r, loc=<21.6303 0 0>, rot=<0.0013 0 0 -1>}
 deform_groin_base_r {parent=thigh_r, loc=<2.1908 -8.1142 1.7442>, rot=<0 0 0 -1>}
 deform_groin_r {parent=deform_groin_base_r, loc=<0 0 0>, rot=<0 0 0 1>}
 deform_glute_base_l {parent=pelvis, loc=<-8.8064 -8.8222 -8.4287>, rot=<0.0275 0.0501 -0.4472 -0.8926>}
 deform_glute_l {parent=deform_glute_base_l, loc=<0 0 0>, rot=<0 0 0 1>}
 deform_glute_base_r {parent=pelvis, loc=<-8.8062 -8.8222 8.4287>, rot=<0.0501 -0.0275 0.8926 -0.4472>}
 deform_glute_r {parent=deform_glute_base_r, loc=<0 0 0>, rot=<0 0 0 1>}
 dyn_arm_l_01_null {type=Null, loc=<17.7788 -4.8444 134.4501>, rot=<0.0134 -0.3555 0.014 -0.9345>}
 dyn_arm_l_01_ctrl {controlType=EULER_TRANSFORM, parent=dyn_arm_l_01_null, type=Control, loc=<0 0 0>, rot=<0 0 0 1>}
 dyn_arm_l_02_null {type=Null, loc=<22.3009 -5.0598 130.4301>, rot=<0.0134 -0.3555 0.014 -0.9345>}
 dyn_arm_l_02_ctrl {controlType=EULER_TRANSFORM, parent=dyn_arm_l_02_null, type=Control, loc=<16.2465 -5.0599 130.4301>, rot=<-0.0134 0.3555 -0.014 0.9345>}
 dyn_arm_l_03_null {type=Null, loc=<27.4606 -5.3057 125.8433>, rot=<0.0134 -0.3555 0.014 -0.9345>}
 dyn_arm_l_03_ctrl {controlType=EULER_TRANSFORM, parent=dyn_arm_l_03_null, type=Control, loc=<20.5525 -5.3057 125.8433>, rot=<-0.0134 0.3555 -0.014 0.9345>}
 dyn_arm_l_04_null {type=Null, loc=<32.9479 -5.5672 120.9652>, rot=<0.0134 -0.3555 0.014 -0.9345>}
 dyn_arm_l_04_ctrl {controlType=EULER_TRANSFORM, parent=dyn_arm_l_04_null, type=Control, loc=<25.6012 -5.5672 120.9652>, rot=<-0.0134 0.3555 -0.014 0.9345>}
 dyn_arm_l_05_null {type=Null, loc=<37.6848 -5.7929 116.7542>, rot=<0.0189 0.3525 0.2369 0.9051>}
 dyn_arm_l_05_ctrl {controlType=EULER_TRANSFORM, parent=dyn_arm_l_05_null, type=Control, loc=<20.3441 -33.1023 114.6204>, rot=<0.1574 0.3181 0.0281 0.9345>}
 dyn_arm_l_06_null {type=Null, loc=<41.6902 -3.0229 112.8126>, rot=<-0.0189 -0.3525 -0.2369 -0.9051>}
 dyn_arm_l_06_ctrl {controlType=EULER_TRANSFORM, parent=dyn_arm_l_06_null, type=Control, loc=<35.425 -3.0229 112.8126>, rot=<0.0189 0.3525 0.2369 0.9051>}
 dyn_arm_l_07_null {type=Null, loc=<45.3705 -0.4776 109.1908>, rot=<0.0189 0.3525 0.2369 0.9051>}
 dyn_arm_l_07_ctrl {controlType=EULER_TRANSFORM, parent=dyn_arm_l_07_null, type=Control, loc=<39.6137 -0.4776 109.1908>, rot=<0.0189 0.3525 0.2369 0.9051>}
 dyn_arm_l_08_null {type=Null, loc=<48.6543 1.7933 105.9592>, rot=<-0.0189 -0.3525 -0.2369 -0.9051>}
 dyn_arm_l_08_ctrl {controlType=EULER_TRANSFORM, parent=dyn_arm_l_08_null, type=Control, loc=<43.5178 1.7933 105.9592>, rot=<0.0189 0.3525 0.2369 0.9051>}
 dyn_arm_l_09_null {type=Null, loc=<51.9727 4.0883 102.6935>, rot=<0.0189 0.3525 0.2369 0.9051>}
 dyn_arm_l_09_ctrl {controlType=EULER_TRANSFORM, parent=dyn_arm_l_09_null, type=Control, loc=<46.782 4.0883 102.6935>, rot=<0.0189 0.3525 0.2369 0.9051>}
 dyn_arm_r_01_null {type=Null, loc=<-17.7788 -4.8444 134.45>, rot=<-0.355 -0.0085 0.9345 -0.0223>}
 dyn_arm_r_01_ctrl {controlType=EULER_TRANSFORM, parent=dyn_arm_r_01_null, type=Control, loc=<0 0 0>, rot=<0 0 0 1>}
 dyn_arm_r_02_null {type=Null, loc=<-22.3008 -5.0599 130.4299>, rot=<0.355 0.0085 -0.9345 0.0223>}
 dyn_arm_r_02_ctrl {controlType=EULER_TRANSFORM, parent=dyn_arm_r_02_null, type=Control, loc=<-28.3553 -5.0599 130.4299>, rot=<0.355 0.0085 -0.9345 0.0223>}
 dyn_arm_r_03_null {type=Null, loc=<-27.4606 -5.3057 125.843>, rot=<-0.355 -0.0085 0.9345 -0.0223>}
 dyn_arm_r_03_ctrl {controlType=EULER_TRANSFORM, parent=dyn_arm_r_03_null, type=Control, loc=<-34.3688 -5.3057 125.843>, rot=<0.355 0.0085 -0.9345 0.0223>}
 dyn_arm_r_04_null {type=Null, loc=<-32.9479 -5.5672 120.965>, rot=<0.355 0.0085 -0.9345 0.0223>}
 dyn_arm_r_04_ctrl {controlType=EULER_TRANSFORM, parent=dyn_arm_r_04_null, type=Control, loc=<-40.2946 -5.5671 120.965>, rot=<0.355 0.0085 -0.9345 0.0223>}
 dyn_arm_r_05_null {type=Null, loc=<-37.6848 -5.7929 116.754>, rot=<-0.3696 -0.0885 0.901 0.2091>}
 dyn_arm_r_05_ctrl {controlType=EULER_TRANSFORM, parent=dyn_arm_r_05_null, type=Control, loc=<-43.0297 -25.9263 114.3665>, rot=<-0.3646 -0.1754 0.9142 -0.0223>}
 dyn_arm_r_06_null {type=Null, loc=<-41.6902 -3.0229 112.813>, rot=<0.3697 0.0885 -0.901 -0.2091>}
 dyn_arm_r_06_ctrl {controlType=EULER_TRANSFORM, parent=dyn_arm_r_06_null, type=Control, loc=<-47.9681 -3.0232 112.8074>, rot=<-0.3697 -0.0885 0.901 0.2091>}
 dyn_arm_r_07_null {type=Null, loc=<-45.3704 -0.4774 109.1911>, rot=<-0.3697 -0.0885 0.901 0.2091>}
 dyn_arm_r_07_ctrl {controlType=EULER_TRANSFORM, parent=dyn_arm_r_07_null, type=Control, loc=<-51.131 -0.4772 109.1893>, rot=<-0.3697 -0.0885 0.901 0.2091>}
 dyn_arm_r_08_null {type=Null, loc=<-48.654 1.7939 105.9591>, rot=<0.3696 0.0885 -0.901 -0.2091>}
 dyn_arm_r_08_ctrl {controlType=EULER_TRANSFORM, parent=dyn_arm_r_08_null, type=Control, loc=<-53.7737 1.7934 105.9678>, rot=<-0.3695 -0.0885 0.9011 0.2091>}
 dyn_arm_r_09_null {type=Null, loc=<-51.9725 4.0888 102.6941>, rot=<-0.3696 -0.0885 0.901 0.2091>}
 dyn_arm_r_09_ctrl {controlType=EULER_TRANSFORM, parent=dyn_arm_r_09_null, type=Control, loc=<-57.1627 4.0888 102.6941>, rot=<-0.3696 -0.0885 0.901 0.2091>}
]
variables [
 arm_l_controls {
  type="TArray<FRigElementKey>"
  default="((Type=Control,Name=\"dyn_arm_l_01_ctrl\"),(Type=Control,Name=\"dyn_arm_l_02_ctrl\"),(Type=Control,Name=\"dyn_arm_l_03_ctrl\"),(Type=Control,Name=\"dyn_arm_l_04_ctrl\"),(Type=Control,Name=\"dyn_arm_l_05_ctrl\"),(Type=Control,Name=\"dyn_arm_l_06_ctrl\"),(Type=Control,N ...(truncated)"
 }
 arm_r_controls {
  type="TArray<FRigElementKey>"
  default="((Type=Control,Name=\"dyn_arm_r_01_ctrl\"),(Type=Control,Name=\"dyn_arm_r_02_ctrl\"),(Type=Control,Name=\"dyn_arm_r_03_ctrl\"),(Type=Control,Name=\"dyn_arm_r_04_ctrl\"),(Type=Control,Name=\"dyn_arm_r_05_ctrl\"),(Type=Control,Name=\"dyn_arm_r_06_ctrl\"),(Type=Control,N ...(truncated)"
 }
]
