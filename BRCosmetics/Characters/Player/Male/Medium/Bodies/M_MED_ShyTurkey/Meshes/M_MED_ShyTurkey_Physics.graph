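# Physics asset for the M_MED_ShyTurkey male medium player body, covering its dangling bow and string bones.
# Capsule bodies form mirrored left and right chains, each side having a two-bone bow and a three-bone string.
# Each chain's root is constrained to spine_05 and later links to their predecessor, all sharing modest swing and tight twist limits.
# Larger capsules on spine_04 and spine_05 give the upper torso volume, with no constraint between them.
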PhysicsAsset /BRCosmetics/Characters/Player/Male/Medium/Bodies/M_MED_ShyTurkey/Meshes/M_MED_ShyTurkey_Physics
bodies {
 dyn_L_bow_1 [{
   shapeType=Capsule
   length=3.1239
   radius=1.3761
   shapeName=dyn_L_bow_1_capsule
   center=<1.5 1.4115 0.0344>
   rotation=<rot 6.0796 92.6892 90.3495>
  }]
 dyn_L_bow_2 [{
   shapeType=Capsule
   length=2.7885
   radius=1.7115
   shapeName=dyn_L_bow_2_capsule
   center=<1.5 1.5912 -0.1131>
   rotation=<rot 7.0183 90.0758 86.8885>
  }]
 dyn_L_string_1 [{
   shapeType=Capsule
   length=2.6359
   radius=1.8641
   shapeName=dyn_L_string_1_capsule
   center=<1.5 -0.9486 1.2295>
   rotation=<rot 65.6162 -90.0392 -90.0002>
  }]
 dyn_L_string_2 [{
   shapeType=Capsule
   length=2.7261
   radius=1.7739
   shapeName=dyn_L_string_2_capsule
   center=<1.5 -0.9486 1.2295>
   rotation=<rot 65.6162 -90.0392 -90.0002>
  }]
 dyn_L_string_3 [{
   shapeType=Capsule
   length=2.7639
   radius=1.7361
   shapeName=dyn_L_string_3_capsule
   center=<1.5 -0.9483 1.2289>
   rotation=<rot 65.4498 -82.9362 -83.5337>
  }]
 dyn_R_bow_1 [{
   shapeType=Capsule
   length=2.7677
   radius=1.7323
   shapeName=dyn_R_bow_1_capsule
   center=<1.5 1.0098 1.1868>
   rotation=<rot 81.7029 178.9413 179.9996>
  }]
 dyn_R_bow_2 [{
   shapeType=Capsule
   length=2.9205
   radius=1.5795
   shapeName=dyn_R_bow_2_capsule
   center=<1.5 0.7277 1.6085>
   rotation=<rot 53.3127 91.5533 98.7616>
  }]
 dyn_R_string_1 [{
   shapeType=Capsule
   length=2.3322
   radius=2.1678
   shapeName=dyn_R_string_1_capsule
   center=<1.5 1.1634 1.3098>
   rotation=<rot 54.0945 90.0002 90.0002>
  }]
 dyn_R_string_2 [{
   shapeType=Capsule
   length=2.2448
   radius=2.2552
   shapeName=dyn_R_string_2_capsule
   center=<1.5 1.2064 1.3692>
   rotation=<rot 54.0946 90.0003 90.0003>
  }]
 dyn_R_string_3 [{
   shapeType=Capsule
   length=2.3409
   radius=2.1591
   shapeName=dyn_R_string_3_capsule
   center=<1.5 1.262 1.4461>
   rotation=<rot 54.0521 86.7989 87.4076>
  }]
 spine_04 [{shapeType=Capsule, length=6.2551, radius=10.4327, shapeName=spine_04_capsule, center=<8.2423 -0.1986 0>, rotation=<rot 0 89.3649 0>}]
 spine_05 [{shapeType=Capsule, length=8.1122, radius=9.857, shapeName=spine_05_capsule, center=<-0.1471 -0.1074 0>, rotation=<rot 0 89.3649 0>}]
}
constraints [
 {bone1Name=dyn_L_bow_1, bone2Name=spine_05, swing1LimitDegrees=20, swing2LimitDegrees=20, twistLimitDegrees=4}
 {bone1Name=dyn_L_bow_2, bone2Name=dyn_L_bow_1, swing1LimitDegrees=20, swing2LimitDegrees=20, twistLimitDegrees=4}
 {bone1Name=dyn_L_string_2, bone2Name=dyn_L_string_1, swing1LimitDegrees=20, swing2LimitDegrees=20, twistLimitDegrees=4}
 {bone1Name=dyn_L_string_1, bone2Name=spine_05, swing1LimitDegrees=20, swing2LimitDegrees=20, twistLimitDegrees=4}
 {bone1Name=dyn_L_string_3, bone2Name=dyn_L_string_2, swing1LimitDegrees=20, swing2LimitDegrees=20, twistLimitDegrees=4}
 {bone1Name=dyn_R_bow_1, bone2Name=spine_05, swing1LimitDegrees=20, swing2LimitDegrees=20, twistLimitDegrees=4}
 {bone1Name=dyn_R_bow_2, bone2Name=dyn_R_bow_1, swing1LimitDegrees=20, swing2LimitDegrees=20, twistLimitDegrees=4}
 {bone1Name=dyn_R_string_1, bone2Name=spine_05, swing1LimitDegrees=20, swing2LimitDegrees=20, twistLimitDegrees=4}
 {bone1Name=dyn_R_string_2, bone2Name=dyn_R_string_1, swing1LimitDegrees=20, swing2LimitDegrees=20, twistLimitDegrees=4}
 {bone1Name=dyn_R_string_3, bone2Name=dyn_R_string_2, swing1LimitDegrees=20, swing2LimitDegrees=20, twistLimitDegrees=4}
]
