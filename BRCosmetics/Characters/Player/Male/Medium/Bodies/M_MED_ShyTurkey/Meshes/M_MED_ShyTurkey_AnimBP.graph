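# Animation Blueprint for the ShyTurkey male body cosmetic, adding secondary motion to the pose copied from its attached parent mesh.
# A Control Rig feeds three chained RigidBody sims for body, left arm and right arm, each with its own physics asset.
# Updates store gravity and rigid body settings from a parameter asset.
# Arm sim alphas drop for hidden arms during dual-wield melee or unarmed stances, and are scaled down in the front end.
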AnimBlueprint /BRCosmetics/Characters/Player/Male/Medium/Bodies/M_MED_ShyTurkey/Meshes/M_MED_ShyTurkey_AnimBP
target_skeleton=&/M_MED_ShyTurkey_Skeleton
graphs [
 AnimGraph {nodes [
   n1 = Control Rig(Source=Copy Pose From Mesh() {bUseAttachedParent=true}) {
    alphaBoolBlend {blendOption=Linear}
    controlRigAssetReference {blueprintRigClass=&/M_MED_ShyTurkey_CtrlRig.M_MED_ShyTurkey_CtrlRig_C}
   }
   n2 = RigidBody!(
    ComponentPose=Local To Component(LocalPose=@n1)
    Alpha
    OverrideWorldGravity:Vector="0,0,0"
    ExternalForce
    SimSpaceSettings:Sim Space Settings Structure=(WorldAlpha=0,VelocityScaleZ=1,DampingAlpha=1,MaxLinearVelocity=10000,MaxAngularVelocity=10000,MaxLinearAcceleration=10000,MaxAngularAcceleration=10000,ExternalLinearDragV=(X=0,Y=0,Z=0),ExternalLinearVelocity=(X=0,Y=0,Z=0),ExternalAngularVelocity=(X=0,Y=0,Z=0))
   ) -> (Pose) {
    bOverrideWorldGravity=true
    overridePhysicsAsset=&/M_MED_ShyTurkey_Physics
    simulationSpace=BaseBoneSpace
    alphaBoolBlend {blendOption=Linear}
    baseBoneRef {boneName=spine_05}
   }
   n3 = RigidBody(
    ComponentPose=@n2
    Alpha=Get Arm_L_RBAN Alpha() -> (Arm_L_RBAN Alpha:double)
    OverrideWorldGravity="0,0,0"
    SimSpaceSettings=(WorldAlpha=0,VelocityScaleZ=1,DampingAlpha=1,MaxLinearVelocity=10000,MaxAngularVelocity=10000,MaxLinearAcceleration=10000,MaxAngularAcceleration=10000,ExternalLinearDragV=(X=0,Y=0,Z=0),ExternalLinearVelocity=(X=0,Y=0,Z=0),ExternalAngularVelocity=(X=0,Y=0,Z=0))
   ) {
    bOverrideWorldGravity=true
    overridePhysicsAsset=&/M_MED_ShyTurkey_Arm_L_Physics
    simulationSpace=BaseBoneSpace
    alphaBoolBlend {blendOption=Linear}
    baseBoneRef {boneName=upperarm_l}
   }
   n4 = RigidBody(
    ComponentPose=@n3
    Alpha=Get Arm_R_RBAN Alpha() -> (Arm_R_RBAN Alpha:double)
    OverrideWorldGravity="0,0,0"
    SimSpaceSettings=(WorldAlpha=0,VelocityScaleZ=1,DampingAlpha=1,MaxLinearVelocity=10000,MaxAngularVelocity=10000,MaxLinearAcceleration=10000,MaxAngularAcceleration=10000,ExternalLinearDragV=(X=0,Y=0,Z=0),ExternalLinearVelocity=(X=0,Y=0,Z=0),ExternalAngularVelocity=(X=0,Y=0,Z=0))
   ) {
    bOverrideWorldGravity=true
    overridePhysicsAsset=&/M_MED_ShyTurkey_Arm_R_Physics
    simulationSpace=BaseBoneSpace
    alphaBoolBlend {blendOption=Linear}
    baseBoneRef {boneName=upperarm_r}
   }
   Output Pose(Result=Component To Local(ComponentPose=@n4)) {layerGroup=None}
  ]}
 EventGraph {nodes [
   Event BlueprintUpdateAnimation()
   > n1 = Sequence!(execute) -> (then_0, then_1, then_2:Exec, then_3:Exec, then_4:Exec)
   > n2 = "Gravity Override State Function (Parameter Asset) (Nativized)"(
    CharacterPartAnimInstance=Self-Reference()
    GameTimeInSeconds=GetGameTimeInSeconds()
    Parameters=&/M_MED_ShyTurkey_Parameters
   )
   > Set Gravity Override(Gravity Override=@n2)
   > Set RBAN Settings Out(RBAN Settings Out=@n2.OutRigidBodyAnimNodeParameters)
   n3 = Cast To FortPlayerAnimInstance(execute=@n1.then_1, Object=GetAnimInstance(self=GetOwnerSkeletalMeshComponent()))
   > Set Is Gauntlet(Is Gauntlet:bool=false) -> (Output_Get:bool)
   > n4 = Switch on EFortWeaponCoreAnimation!(execute, Selection:EFortWeaponCoreAnimation Enum=Get WeaponCoreAnim(self=@n3)) -> (
    Melee:Exec
    Pistol:Exec
    Shotgun:Exec
    PaperBlueprint:Exec
    Rifle:Exec
    MeleeOneHand:Exec
    MachinePistol:Exec
    RocketLauncher:Exec
    GrenadeLauncher:Exec
    GoingCommando:Exec
    AssaultRifle:Exec
    TacticalShotgun:Exec
    SniperRifle:Exec
    TrapPlacement:Exec
    ShoulderLauncher:Exec
    AbilityDecoTool:Exec
    Crossbow:Exec
    C4:Exec
    RemoteControl:Exec
    DualWield:Exec
    AR_BullPup:Exec
    AR_ForwardGrip:Exec
    MedPackPaddles:Exec
    SMG_P90:Exec
    AR_DrumGun:Exec
    Consumable_Small:Exec
    Consumable_Large:Exec
    Balloon:Exec
    MountedTurret:Exec
    CreativeTool:Exec
    ExplosiveBow:Exec
    AshtonIndigo:Exec
    AshtonChicago:Exec
    MeleeDualWield:Exec
    Unarmed:Exec
   )
   Set Is Gauntlet(execute=[@n4.MeleeDualWield, @n4.Unarmed], Is Gauntlet=true)
   Branch(execute=@n1.then_3, Condition=Get bIsInFrontEnd())
   > Set Arm_L_RBAN Alpha(Arm_L_RBAN Alpha:double=float * float(A=Get Arm_L_RBAN Alpha(), B=0.25) {is=title K2Node_PromotableOperator}) -> (Output_Get:double)
   > Set Arm_R_RBAN Alpha(Arm_R_RBAN Alpha:double=float * float(A=Get Arm_R_RBAN Alpha(), B=0.25)) -> (Output_Get:double)
   n5 = Branch(
    execute=@n1.then_2
    Condition=AND Boolean!(
     A=Get Is Gauntlet() -> (Is Gauntlet:bool)
     B=Get bIsLeftArmHidden() -> (bIsLeftArmHidden:bool)
     C:bool=Get bIsRightArmHidden() -> (bIsRightArmHidden:bool)
    ) -> (ReturnValue)
   )
   > Set Arm_L_RBAN Alpha(Arm_L_RBAN Alpha=0.25)
   > Set Arm_R_RBAN Alpha(Arm_R_RBAN Alpha=0.25)
   Set Arm_L_RBAN Alpha(execute=@n5.else, Arm_L_RBAN Alpha=1)
   > Set Arm_R_RBAN Alpha(Arm_R_RBAN Alpha=1)
   > n6 = Branch(Condition=AND Boolean!(A=Get Is Gauntlet(), B=Get bIsLeftArmHidden()) -> (ReturnValue))
   > Set Arm_L_RBAN Alpha(Arm_L_RBAN Alpha=0.25)
   Set Arm_L_RBAN Alpha(execute=@n6.else, Arm_L_RBAN Alpha=1)
   > n7 = Branch(Condition=AND Boolean(A=Get Is Gauntlet(), B=Get bIsRightArmHidden()))
   > Set Arm_R_RBAN Alpha(Arm_R_RBAN Alpha=0.25)
   Set Arm_R_RBAN Alpha(execute=@n7.else, Arm_R_RBAN Alpha=1)
  ]}
]
variables [
 Arm_L_RBAN Alpha {type=number}
 Arm_R_RBAN Alpha {type=number}
 Gravity Override {type=Vector}
 Is Gauntlet {type=boolean}
 RBAN Settings Out {type=FortRigidBodyAnimNodeParameters}
]
